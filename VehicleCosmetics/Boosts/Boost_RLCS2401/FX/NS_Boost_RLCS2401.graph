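# Niagara system for the RLCS2401 vehicle boost cosmetic, drawing ribbon trails from the car's boost ports.
# One emitter spawns a ribbon per port, positions it by skeletal position plus a car offset, and pushes particles along negative X.
# Painted-variant overrides set two color layers, each faded by its own alpha curve over the particle's life.
# Four ribbon renderers draw the trail, with the painted flag passed through dynamic material parameters and ribbon attributes.
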NiagaraSystem /VehicleCosmetics/Boosts/Boost_RLCS2401/FX/NS_Boost_RLCS2401
dependencies {
 usedDataInterfaces [&NiagaraDataInterfaceCurve, &NiagaraDataInterfaceSkeletalMesh]
 usedDynamicInputs []
 usedModules [
  &self:SystemSpawnScript.SystemScriptSource.SystemScriptGraph.NiagaraNodeAssignment_0.SetVariables_#1
  &SystemState
 ]
}
emitters {NE_VC_Boost_Ribbon_Parent {
  emitterUpdateScript [
   &EmitterState()
   &VC_FilteredSocketCount()
   &VC_RateQualityHandler()
   &VC_SpawnRibbon_v1(Number of Ribbons=$Emitter.PortCount, SpawnRate=$Emitter.Spawn Rate)
  ]
  particleSpawnScript [
   &VC_PaintedVariantsOverride(
    BurntSienna=<rgba 0.0642 0.0198 0.0079 1>
    DefaultColor=<rgba 1.05 0.9412 0.0164 1>
    R1 Black:LinearColor=<rgba 0.0191 0.0191 0.0191 0>
    R1 BurntSienna:LinearColor=<rgba 0.0642 0.0198 0.0079 0>
    R1 Cobalt:LinearColor=<rgba 0 0.0185 1 0>
    R1 Crimson:LinearColor=<rgba 0.4635 0 0.0126 0>
    R1 DefaultColor=<rgba 0.1732 1.2 0.4218 2.2>
    R1 ForestGreen:LinearColor=<rgba 0 0.5 0.0253 0>
    R1 Grey:LinearColor=<rgba 0.2031 0.2031 0.2031 0>
    R1 Lime:LinearColor=<rgba 0.2943 0.5885 0 0>
    R1 Orange:LinearColor=<rgba 0.6724 0.2346 0 0>
    R1 Pink:LinearColor=<rgba 0.5052 0.0605 0.3435 0>
    R1 Purple:LinearColor=<rgba 0.2789 0 0.6795 0>
    R1 Saffron:LinearColor=<rgba 0.9 0.69 0 0>
    R1 SkyBlue:LinearColor=<rgba 0 0.357 0.5365 0>
    R1 TitaniumWhite:LinearColor=<rgba 0.75 0.75 0.75 0>
    R2 Black:LinearColor=<rgba 0.0208 0.0208 0.0208 3.5>
    R2 BurntSienna:LinearColor=<rgba 0.1094 0.0362 0.0165 6.25>
    R2 Cobalt:LinearColor=<rgba 0.0137 0.0215 0.4375 7>
    R2 Crimson:LinearColor=<rgba 0.2969 0.0247 0.0321 7>
    R2 DefaultColor:LinearColor=<rgba 0 12 0 0.015>
    R2 ForestGreen:LinearColor=<rgba 0.1151 0.2865 0 6.5>
    R2 Grey:LinearColor=<rgba 0.2083 0.2083 0.2083 6.5>
    R2 Lime:LinearColor=<rgba 0.3432 0.5156 0.1182 3>
    R2 Orange:LinearColor=<rgba 0.375 0.2289 0.1309 6.5>
    R2 Pink:LinearColor=<rgba 0.3802 0.103 0.2794 6.6>
    R2 Purple:LinearColor=<rgba 0.2111 0.0139 0.3333 8>
    R2 Saffron:LinearColor=<rgba 0.3385 0.2715 0.0511 4>
    R2 SkyBlue:LinearColor=<rgba 0 0.216 0.2552 8>
    R2 TitaniumWhite:LinearColor=<rgba 0.3229 0.3229 0.3229 15>
    Use R1 Painted Overrides=true
    Use R2 Painted Overrides=true
   )
   &VC_InitializeRibbon_v1(
    Color=$Particles.Color
    Custom Offset=true
    Enable Skeletal Position=true
    Lifetime=1
    Local Vector Offset=&AddVectorToPosition(Position=<0 0 0>, Vector=$User.DTCarOffset)
    Offset Amount=1
    Ribbon Width=150
   )
   &AddVelocity(Velocity=<-1400 0 0>) {module=AddVelocity001}
   /SetVariables(Particles.ColorTag02:LinearColor=$Particles.R2 Color, Particles.ColorTag1:LinearColor=$Particles.R1 Color) {isSetParameters=true}
   /SetVariables(Particles.R1Main=150, Particles.R2Lines=85, Particles.R3BackGround=130) {isSetParameters=true}
  ]
  particleUpdateScript [
   &ParticleState()
   &InheritSourceMovement(ApplyForce=false, ApplyPosition=true, ApplyPositionFalloff=false)
   &SolveForcesAndVelocity(Clamp Velocity=false, Limit Acceleration=false)
   &VC_UpdateRibbon_v1(Enable Skeletal Position=true)
   &ScaleColor(
    Scale Alpha=&FloatFromCurve(FloatCurve=DataInterface() {propertyValues {Curve {keys [<at 0 0>,<at 0.2 2 RCIM_Cubic>,<at 1 0>]}}})
    Scale RGB=&VectorFromFloat(Value=1)
    ScaleA=true
    ScaleRGB=true
   )
   &DynamicMaterialParameters(
    Index 0 Param 1=&MakeCustomFloatFromBool(Bool=$User.bIsPainted, True Float=1, False Float=0)
    Index 0 Param 2=1.4
    Index 0 Param 3=2.85
    Index 0 Param 4=2
    Module.Index0Float0Write=true
    Module.Index0Float1Write=true
    Module.Index0Float2Write=true
    Module.Index0Float3Write=true
   )
   /SetVariables(
    Particles.Ribbon1Index0:Vector4f=&MakeVector4(X=&MakeCustomFloatFromBool(Bool=$User.bIsPainted, True Float=1, False Float=0), Y=1, Z=0, W=1)
    Particles.RibbonColor1:LinearColor=&VC_DynamicScaleColor(
     Color to Scale:LinearColor=$Particles.Initial.ColorTag1
     RGB Scale:Vector3f=&VectorFromFloat(Value=1)
     Alpha Scale:NiagaraFloat=&FloatFromCurve(FloatCurve=DataInterface() {propertyValues {Curve {keys [
          <at 0 0 RCIM_Cubic>
          {arriveTangentWeight=0.0673, interpMode=RCIM_Cubic, leaveTangentWeight=0.101, tangentMode=RCTM_User, time=0.1, value=0.25}
          {arriveTangent=-0.3125, interpMode=RCIM_Cubic, leaveTangent=-0.3125, time=0.5, value=0.55}
          <at 0.9 0 RCIM_Cubic>
         ]}}})
    )
    Particles.RibbonColor2:LinearColor=&VC_DynamicScaleColor(
     Color to Scale:LinearColor=$Particles.ColorTag02
     RGB Scale:Vector3f=&VectorFromFloat(Value=1)
     Alpha Scale:NiagaraFloat=&FloatFromCurve(FloatCurve=DataInterface() {propertyValues {Curve {keys [
          <at 0 0 RCIM_Cubic>
          {arriveTangentWeight=0.0673, interpMode=RCIM_Cubic, leaveTangentWeight=0.101, tangentMode=RCTM_User, time=0.1, value=0.7}
          {arriveTangent=-0.875, interpMode=RCIM_Cubic, leaveTangent=-0.875, time=0.5, value=1}
          <at 0.9 0 RCIM_Cubic>
         ]}}})
    )
   ) {isSetParameters=true}
  ]
  rendererClasses [&NiagaraRibbonRendererProperties]
  renderers [
   {rendererClass=&NiagaraRibbonRendererProperties}
   {rendererClass=&NiagaraRibbonRendererProperties, rendererIndex=1}
   {rendererClass=&NiagaraRibbonRendererProperties, rendererIndex=2}
   {rendererClass=&NiagaraRibbonRendererProperties, rendererIndex=3}
  ]
 }}
summary {
 systemName=NS_Boost_RLCS2401
 emitters [{emitterName=NE_VC_Boost_Ribbon_Parent, rendererClasses [&NiagaraRibbonRendererProperties]}]
 userVariables [
  User.bIsFrontendPreview {type=NiagaraBool}
  User.bIsPainted {type=NiagaraBool}
  User.ColorVariantID {type=NiagaraInt32}
  User.DTCarOffset {type=Vector3f, description="", defaultValue {struct=Vector3f, value=<0 0 0>}}
  User.PaintedColor {type=LinearColor, description="", defaultValue {struct=LinearColor, value=<rgba 1 1 1 1>}}
  User.PaintedScalar {type=NiagaraFloat, defaultValue=1, description=""}
  User.PortCount {type=NiagaraInt32, defaultValue=1}
 ]
}
