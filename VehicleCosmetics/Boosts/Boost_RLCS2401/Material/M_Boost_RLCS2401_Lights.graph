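# Unlit translucent material for Niagara ribbons, drawing the animated light streaks of the RLCS2401 boost cosmetic.
# One channel-packed texture, offset, rotatable and panned along U, supplies emissive layers that time-panned cosine bands boost into travelling pulses.
# Particle color tints the glow, while dynamic parameters drive offset, overall glow, outline glow and alpha.
# An alpha mask combines texture and particle alpha with a U gradient that fades both ends.
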Material /VehicleCosmetics/Boosts/Boost_RLCS2401/Material/M_Boost_RLCS2401_Lights
graph {
 properties {
  bAutomaticallySetUsageInEditor=false
  bUsedWithNiagaraRibbons=true
  blendMode=BLEND_Translucent
  shadingModel=MSM_Unlit
  usedShadingModels=Unlit
 }
 nodes [
  MaterialExpressionVertexColor_0 = ParticleColor()
  n1 = DynamicParameter() {paramNames [Offset,General_Glow,Glow_Outline,AlphaGlow]}
  n2 = Constant()
  n3 = MaterialFunctionCall(
   UVs=Add(A=TextureCoordinate() {uTiling=2}, B=AppendVector(A=@n1.Offset, B=@n2))
   "Rotation Angle (0-1)"=ScalarParameter() {parameterName=RotateTex}
  ) {materialFunction=&CustomRotator}
  n4 = TextureSample(UVs=Panner(Coordinate=@n3.Rotated Values) {speedX=1.5}) {texture=/DealershipSharedLibrary/FX/Textures/TexturePackage/T_VCVFX_Red}
  n5 = Multiply(A=@MaterialExpressionVertexColor_0.A, B=@n4.A)
  n6 = ComponentMask(None=TextureCoordinate()) {r=true}
  n7 = Constant() {r=1.2}
  n8 = OneMinus(None=@n6)
  n9 = Constant() {r=2}
  n10 = Multiply(A=Multiply(A=Power(Base=@n6, Exp=@n7), B=Constant() {r=5}), B=Multiply(A=Power(Base=@n8, Exp=@n9), B=Constant() {r=1.5}))
  MaterialExpressionConstantClamp_0 = Saturate(None=Multiply(A=Multiply(A=@n5, B=@n10), B=@n1.AlphaGlow))
  n11 = TextureCoordinate() {uTiling=0.4}
  n12 = Time()
  n13 = ComponentMask(None=Cosine(None=Panner(Coordinate=@n11, Time=@n12) {speedX=0.6})) {r=true}
  MaterialExpressionConstantClamp_1 = Saturate(None=Add(A=@n13, B=Constant() {r=0.5}))
  n14 = ComponentMask(None=@n13) {r=true}
  MaterialExpressionConstantClamp_2 = Saturate(None=Add(A=OneMinus(None=@n14), B=Constant() {r=-0.8}))
  n15 = Multiply(A=Multiply(A=@MaterialExpressionConstantClamp_1, B=@MaterialExpressionConstantClamp_2), B=Constant() {r=4})
  n16 = TextureSample(UVs=Panner(Coordinate=@n3.Rotated Values) {speedX=1.5}) {texture=&T_VCVFX_Red}
  n17 = Constant() {r=20}
  n18 = Multiply(A=Multiply(A=@n15, B=Desaturation(None=@n16.G, Fraction=@n17)), B=Constant() {r=2.5})
  n19 = Panner(Coordinate=@n3.Rotated Values) {speedX=1.5}
  n20 = TextureCoordinate() {uTiling=0.4}
  n21 = Time()
  n22 = ComponentMask(None=Cosine(None=Panner(Coordinate=@n20, Time=@n21) {speedX=0.6})) {r=true}
  MaterialExpressionConstantClamp_3 = Saturate(None=Add(A=@n22, B=Constant() {r=0.5}))
  n23 = ComponentMask(None=@n22) {r=true}
  MaterialExpressionConstantClamp_4 = Saturate(None=Add(A=OneMinus(None=@n23), B=Constant() {r=-0.8}))
  n24 = Multiply(
   A=Desaturation(None=TextureSample(UVs=@n19) {texture=&T_VCVFX_Red}.R, Fraction=Constant() {r=15})
   B=Multiply(A=Multiply(A=@MaterialExpressionConstantClamp_3, B=@MaterialExpressionConstantClamp_4), B=Constant() {r=4})
  )
  n25 = Power(Base=@n13, Exp=Constant() {r=0.6})
  n26 = TextureSample(UVs=Panner(Coordinate=@n3.Rotated Values) {speedX=1.5}) {texture=&T_VCVFX_Red}
  n27 = Add(A=Multiply(A=Multiply(A=@n24, B=@n25), B=Constant() {r=2.5}), B=@n26.R)
  n28 = Multiply(A=Constant() {r=2.5}, B=Multiply(A=@n15, B=Desaturation(None=@n16.B, Fraction=@n17)))
  n29 = Add(A=Add(A=Add(A=@n18, B=@n27), B=Add(A=@n28, B=@n26.G)), B=Multiply(A=Multiply(A=@n1.Glow_Outline, B=@n4.B), B=Constant() {r=1}))
  MaterialFunctionCall(Emissive=Multiply(A=Multiply(A=@n29, B=@MaterialExpressionVertexColor_0.RGB), B=@n1.General_Glow)) {materialFunction=&MF_FX_EmissiveAdjustments}
 ]
}
